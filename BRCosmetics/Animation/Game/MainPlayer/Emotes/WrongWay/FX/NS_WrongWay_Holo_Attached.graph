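# Niagara system for the attached hologram effect of the WrongWay player emote: a short, mostly blue glow with flickering light.
# BlueFlickerLight multiplies a fade envelope by a stepped flicker curve, while Corona drives dynamic material parameters over emitter age.
# Two SubUV_ConeGlow emitters add burst and rate-spawned sprites, the looping one shaped by emitter-age curves over about 2.5 seconds.
# The RayCone, RayCone001 and Empty001 mesh-rendering emitters are disabled.
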NiagaraSystem /BRCosmetics/Animation/Game/MainPlayer/Emotes/WrongWay/FX/NS_WrongWay_Holo_Attached
dependencies {
 usedDataInterfaces [&NiagaraDataInterfaceCurve, &NiagaraDataInterfaceColorCurve, &NiagaraDataInterfaceVector2DCurve]
 usedDynamicInputs []
 usedModules [&SystemState]
}
emitters {
 BlueFlickerLight {
  emitterUpdateScript [
   &EmitterState()
   &SpawnBurst_Instantaneous(Use Spawn Probability=false, UseLoopCountLimit=false)
  ]
  particleSpawnScript [&InitializeParticle(
    Color=<rgba 0.02 0.3467 1 10>
    Lifetime=2.3
    Position=<0 0 40>
    Position Mode=::NewEnumerator1
    UsePositionOffset=false
    Write Material Random=true
   )]
  particleUpdateScript [
   &ParticleState()
   &FN_Light_Attributes(Create Color Attribute=true, Create Radius Attribute=true, Enable Particle Light Enable=true, Light Enabled=true, Radius=50)
   &ScaleColor(
    Scale Alpha=&Multiply_Float(
     A=&FloatFromCurve(FloatCurve=DataInterface() {propertyValues {Curve {keys [
          {arriveTangent=-0.2939, interpMode=RCIM_Cubic, leaveTangent=-0.2939, tangentMode=RCTM_User, time=0, value=0}
          {interpMode=RCIM_Cubic, tangentMode=RCTM_User, time=0.5, value=1}
          {interpMode=RCIM_Cubic, tangentMode=RCTM_User, time=0.8, value=1}
          {arriveTangent=-13.4036, interpMode=RCIM_Cubic, leaveTangent=-13.4036, tangentMode=RCTM_User, time=1, value=0}
         ]}}})
     B=&FloatFromCurve(FloatCurve=DataInterface() {propertyValues {Curve {keys [
          <at 0.0344 0.0746 RCIM_Constant>
          <at 0.0462 1.3837 RCIM_Constant>
          <at 0.0583 0.1032 RCIM_Constant>
          <at 0.0805 1.5625 RCIM_Constant>
          <at 0.0919 0.1154 RCIM_Constant>
          <at 0.1076 1.6916 RCIM_Constant>
          <at 0.1197 0.1237 RCIM_Constant>
          <at 0.1467 1.656 RCIM_Constant>
          <at 0.166 0.178 RCIM_Constant>
          <at 0.1947 1.6562 RCIM_Constant>
          <at 0.2248 0.8073 RCIM_Constant>
          <at 0.7283 0.8445 RCIM_Constant>
          <at 0.7685 0.2588 RCIM_Constant>
          <at 0.7822 1.4836 RCIM_Constant>
          <at 0.7986 0.1766 RCIM_Constant>
          <at 0.8195 1.3828 RCIM_Constant>
          <at 0.8267 0.1814 RCIM_Constant>
          <at 0.8579 1.3759 RCIM_Constant>
          <at 0.8666 0.1699 RCIM_Constant>
          <at 0.9024 0.6778>
          <at 0.928 0.3628 RCIM_Constant>
          {arriveTangent=-2.2618, interpMode=RCIM_Cubic, leaveTangent=-2.2618, time=0.9417, value=0.549}
          {arriveTangent=-13.4036, interpMode=RCIM_Constant, leaveTangent=-13.4036, time=1, value=0.2}
         ]}}})
    )
    ScaleA=true
    ScaleRGB=true
   )
  ]
  rendererClasses [&NiagaraSpriteRendererProperties, &NiagaraLightRendererProperties]
  renderers [
   {rendererClass=&NiagaraSpriteRendererProperties}
   {rendererClass=&NiagaraLightRendererProperties, rendererIndex=1}
  ]
 }
 Corona {
  emitterUpdateScript [
   &EmitterState()
   &SpawnBurst_Instantaneous(Use Spawn Probability=false, UseLoopCountLimit=false)
  ]
  particleSpawnScript [
   &InitializeParticle(
    Color=<rgba 0.3475 1.0721 2 0.4>
    Lifetime=2.5
    Position Mode=::NewEnumerator1
    Sprite Size Mode=::NewEnumerator3
    Uniform Sprite Size=20
    UsePositionOffset=false
    Write Material Random=true
   )
   &DynamicMaterialParameters(
    Index 0 Param 1=1
    Index 0 Param 2=0.7
    Module.Param0WriteEnabled=true
    Module.Param1WriteEnabled=true
    Module.Param2WriteEnabled=true
    Module.Param3WriteEnabled=true
   )
  ]
  particleUpdateScript [
   &ParticleState()
   /SetVariables(Particles.SpriteFacingFlat:Vector3f=<1 0 0>) {isSetParameters=true}
   &CameraOffset(Camera Offset Amount=5)
   &DynamicMaterialParameters(
    Index 0 Param 1=&FloatFromCurve(FloatCurve=DataInterface() {propertyValues {Curve {keys [<at 0 0>,<at 1 1>,<at 2 1>,<at 2.5 0>]}}}, CurveIndex=$Emitter.Age)
    Index 0 Param 2=&FloatFromCurve(FloatCurve=DataInterface() {propertyValues {Curve {keys [<at 0 1>,<at 0.5081 0.7>,<at 1.8035 0.7>,<at 2.0088 1>]}}}, CurveIndex=$Emitter.Age)
    Module.Param0WriteEnabled=true
    Module.Param1WriteEnabled=true
    Module.Param2WriteEnabled=true
    Module.Param3WriteEnabled=true
   ) {module=DynamicMaterialParameters001}
  ]
  rendererClasses [&NiagaraSpriteRendererProperties, &NiagaraLightRendererProperties]
  renderers [
   {rendererClass=&NiagaraSpriteRendererProperties}
   {rendererClass=&NiagaraSpriteRendererProperties, rendererIndex=1}
   {rendererClass=&NiagaraLightRendererProperties, rendererIndex=2}
  ]
 }
 Empty001 {
  bEnabled=false
  emitterUpdateScript [
   &EmitterState()
   &SpawnBurst_Instantaneous(Use Spawn Probability=false, UseLoopCountLimit=false)
  ]
  particleSpawnScript [&InitializeParticle(Color Mode=::NewEnumerator0, Lifetime=5, UsePositionOffset=false, Write Material Random=true)]
  particleUpdateScript [&ParticleState()]
  rendererClasses [&NiagaraMeshRendererProperties]
  renderers [{rendererClass=&NiagaraMeshRendererProperties, rendererIndex=0}]
 }
 RayCone {
  bEnabled=false
  emitterUpdateScript [
   &EmitterState()
   &SpawnBurst_Instantaneous(Use Spawn Probability=false, UseLoopCountLimit=false)
  ]
  particleSpawnScript [&InitializeParticle(Lifetime=5, UsePositionOffset=false, Write Material Random=true)]
  particleUpdateScript [&ParticleState()]
  rendererClasses [&NiagaraSpriteRendererProperties, &NiagaraMeshRendererProperties]
  renderers [
   {rendererClass=&NiagaraSpriteRendererProperties}
   {rendererClass=&NiagaraMeshRendererProperties, rendererIndex=1}
  ]
 }
 RayCone001 {
  bEnabled=false
  emitterUpdateScript [
   &EmitterState()
   &SpawnBurst_Instantaneous(Use Spawn Probability=false, UseLoopCountLimit=false)
  ]
  particleSpawnScript [
   &InitializeParticle(Lifetime=5, UsePositionOffset=false, Write Material Random=true)
   &DynamicMaterialParameters(
    Index 0 Param 1=-250
    Index 0 Param 2=0.5
    Module.Param0WriteEnabled=true
    Module.Param1WriteEnabled=true
    Module.Param2WriteEnabled=true
    Module.Param3WriteEnabled=true
   )
  ]
  particleUpdateScript [&ParticleState()]
  rendererClasses [&NiagaraSpriteRendererProperties, &NiagaraMeshRendererProperties]
  renderers [
   {rendererClass=&NiagaraSpriteRendererProperties}
   {rendererClass=&NiagaraMeshRendererProperties, rendererIndex=1}
  ]
 }
 SubUV_ConeGlow {
  emitterUpdateScript [
   &EmitterState(Life Cycle Mode=::NewEnumerator1, Loop Behavior=::NewEnumerator1, Loop Duration=1, UseLoopDelay=false)
   &SpawnRate(Use Spawn Probability=false) {enabled=false}
   &SpawnBurst_Instantaneous(Use Spawn Probability=false, UseLoopCountLimit=false)
   &SpawnBurst_Instantaneous(Spawn Time=0.04, Use Spawn Probability=false, UseLoopCountLimit=false) {module=SpawnBurst_Instantaneous001}
  ]
  particleSpawnScript [
   &InitializeParticle(
    Color=<rgba 2.7521 2.4219 3 1>
    Lifetime=&RandomRangeFloat(Minimum=0.25, Maximum=0.35)
    UsePositionOffset=false
    Write Material Random=true
   )
   &AddVelocity(Velocity=<1 0 0>)
  ]
  particleUpdateScript [
   &SolveForcesAndVelocity(Clamp Velocity=false, Limit Acceleration=false)
   &ParticleState()
   &ScaleColor(
    Scale Alpha=&FloatFromCurve(FloatCurve=DataInterface() {propertyValues {Curve {keys [
         {arriveTangent=9.0725, interpMode=RCIM_Cubic, leaveTangent=9.0725, tangentMode=RCTM_User, time=0, value=0}
         {arriveTangent=-2.5249, interpMode=RCIM_Cubic, leaveTangent=-2.5249, tangentMode=RCTM_User, time=0.1408, value=0.9847}
         {arriveTangent=-0.0405, interpMode=RCIM_Cubic, leaveTangent=-0.0405, tangentMode=RCTM_User, time=1, value=0}
        ]}}})
    Scale RGB=&MakeVectorFromLinearColorRGB(Color=&ColorFromCurve(ColorCurve=DataInterface() {propertyValues {
        AlphaCurve {keys [<at 0 1 RCIM_Cubic>,<at 1 1 RCIM_Cubic>]}
        BlueCurve {keys [<at 0.4 1 RCIM_Cubic>,<at 0.7 1>]}
        GreenCurve {keys [<at 0.4 0.788 RCIM_Cubic>,<at 0.7 0.232>]}
        RedCurve {keys [<at 0.4 0.168 RCIM_Cubic>,<at 0.7 0.155>]}
       }}))
    ScaleA=true
    ScaleRGB=true
   )
   &ScaleSpriteSize(
    Non-Uniform Curve Sprite Scale=DataInterface() {propertyValues {
      XCurve {keys [
        {arriveTangent=1.7395, interpMode=RCIM_Cubic, leaveTangent=1.7395, tangentMode=RCTM_User, time=0, value=0.5}
        <at 0.4 1 RCIM_Cubic>
        {arriveTangent=-1.343, interpMode=RCIM_Cubic, leaveTangent=-1.343, tangentMode=RCTM_User, time=1, value=0.5}
       ]}
      YCurve {keys [
        {arriveTangent=1.5645, interpMode=RCIM_Cubic, leaveTangent=1.5645, tangentMode=RCTM_User, time=0, value=0.4}
        {arriveTangent=0.8981, interpMode=RCIM_Cubic, leaveTangent=0.8981, tangentMode=RCTM_User, time=0.4, value=1}
        <at 1 1.1 RCIM_Cubic>
       ]}
     }}
    Scale Sprite Size Mode=::NewEnumerator3
   ) {enabled=false}
  ]
  rendererClasses [&NiagaraSpriteRendererProperties]
  renderers [{rendererClass=&NiagaraSpriteRendererProperties}]
 }
 SubUV_ConeGlowLoop {
  emitterUpdateScript [
   &EmitterState()
   &SpawnRate(SpawnRate=5, Use Spawn Probability=false)
   &SpawnBurst_Instantaneous(Use Spawn Probability=false, UseLoopCountLimit=false) {enabled=false}
   &SpawnBurst_Instantaneous(Spawn Time=0.04, Use Spawn Probability=false, UseLoopCountLimit=false) {module=SpawnBurst_Instantaneous001}
  ]
  particleSpawnScript [
   &InitializeParticle(
    Color=<rgba 0.9174 0.8073 1 0.5>
    Lifetime=&RandomRangeFloat(Minimum=1, Maximum=1.25)
    Sprite Size Max=<40 30>
    Sprite Size Min=<30 30>
    Sprite Size Mode=::NewEnumerator2
    UsePositionOffset=false
    Write Material Random=true
   )
   &AddVelocity(Velocity=<1 0 0>)
   &SubUVAnimation(End Frame=0)
  ]
  particleUpdateScript [
   &SolveForcesAndVelocity(Clamp Velocity=false, Limit Acceleration=false)
   &ParticleState()
   &ScaleSpriteSize(
    Non-Uniform Curve Index:NiagaraFloat=$Emitter.Age
    Non-Uniform Curve Sprite Scale=DataInterface() {propertyValues {
      XCurve {keys [
        {arriveTangent=5.8588, interpMode=RCIM_Cubic, leaveTangent=5.8588, tangentMode=RCTM_User, time=0, value=0}
        {arriveTangent=3.727, interpMode=RCIM_Cubic, leaveTangent=3.727, time=0.1308, value=1.1578}
        {arriveTangent=0.0084, interpMode=RCIM_Cubic, leaveTangent=0.0084, time=0.2649, value=0.9874}
        {interpMode=RCIM_Cubic, tangentMode=RCTM_User, time=1.9947, value=1.1734}
        {arriveTangent=-2.3221, interpMode=RCIM_Cubic, leaveTangent=-2.3221, time=2.1459, value=1.6427}
        {arriveTangent=-3.7894, interpMode=RCIM_Cubic, leaveTangent=-3.7894, tangentMode=RCTM_User, time=2.5, value=0}
       ]}
      YCurve {keys [
        {arriveTangent=3.9464, interpMode=RCIM_Cubic, leaveTangent=3.9464, tangentMode=RCTM_User, time=0, value=0.1}
        {interpMode=RCIM_Cubic, tangentMode=RCTM_User, time=0.5619, value=0.8648}
        {arriveTangent=-0.1984, interpMode=RCIM_Cubic, leaveTangent=-0.1984, time=0.752, value=0.6956}
        {arriveTangent=-0.3605, interpMode=RCIM_Cubic, leaveTangent=-0.3605, time=2.3549, value=0.509}
        <at 2.5176 0.059 RCIM_Cubic>
       ]}
     }}
    Scale Sprite Size Mode=::NewEnumerator3
   )
   &ScaleColor(
    Scale Alpha=&FloatFromCurve(
     FloatCurve=DataInterface() {propertyValues {Curve {keys [
         {arriveTangent=11.3841, interpMode=RCIM_Cubic, leaveTangent=11.3841, tangentMode=RCTM_User, time=0, value=0}
         {arriveTangent=3.3333, interpMode=RCIM_Cubic, leaveTangent=3.3333, time=0.1, value=2}
         {arriveTangent=-0.6674, interpMode=RCIM_Cubic, leaveTangent=-0.6674, time=0.3, value=1}
         {arriveTangent=-0.2894, interpMode=RCIM_Cubic, leaveTangent=-0.2894, tangentMode=RCTM_User, time=1.8187, value=0.8529}
         {arriveTangent=2.9353, interpMode=RCIM_Cubic, leaveTangent=2.9353, time=1.9187, value=2.5}
         {arriveTangent=-3.9339, interpMode=RCIM_Cubic, leaveTangent=-3.9339, time=2.0187, value=1.44}
         {arriveTangent=-6.3245, interpMode=RCIM_Cubic, leaveTangent=-6.3245, tangentMode=RCTM_User, time=2.5, value=0.2131}
        ]}}}
     CurveIndex=$Emitter.Age
     Scale Curve=0.2
    )
    Scale RGB=&MakeVectorFromLinearColorRGB(Color=&ColorFromCurve(ColorCurve=DataInterface() {propertyValues {
        AlphaCurve {keys [<at 0 1 RCIM_Cubic>,<at 1 1 RCIM_Cubic>]}
        BlueCurve {keys [<at 0.4 1 RCIM_Cubic>,<at 0.7 1>]}
        GreenCurve {keys [<at 0.4 0.788 RCIM_Cubic>,<at 0.7 0.232>]}
        RedCurve {keys [<at 0.4 0.168 RCIM_Cubic>,<at 0.7 0.155>]}
       }}))
    ScaleA=true
    ScaleRGB=true
   )
   &CameraOffset(Camera Offset Amount=35)
  ]
  rendererClasses [&NiagaraSpriteRendererProperties]
  renderers [{rendererClass=&NiagaraSpriteRendererProperties}]
 }
}
summary {
 systemName=NS_WrongWay_Holo_Attached
 emitters [
  {
   bEnabled=false
   emitterName=RayCone
   rendererClasses [&NiagaraSpriteRendererProperties, &NiagaraMeshRendererProperties]
  }
  {
   bEnabled=false
   emitterName=RayCone001
   rendererClasses [&NiagaraSpriteRendererProperties, &NiagaraMeshRendererProperties]
  }
  {emitterName=Corona, rendererClasses [&NiagaraSpriteRendererProperties, &NiagaraLightRendererProperties]}
  {emitterName=SubUV_ConeGlow, rendererClasses [&NiagaraSpriteRendererProperties]}
  {bEnabled=false, emitterName=Empty001, rendererClasses [&NiagaraMeshRendererProperties]}
  {emitterName=SubUV_ConeGlowLoop, rendererClasses [&NiagaraSpriteRendererProperties]}
  {
   emitterName=BlueFlickerLight
   rendererClasses [&NiagaraSpriteRendererProperties, &NiagaraLightRendererProperties]
  }
 ]
 userVariables [User.Skeletal Mesh {
   type=NiagaraDataInterfaceSkeletalMesh
   defaultValue {
    struct=NiagaraExt_VariableValue_DataInterface
    value {dataInterface=&self:NiagaraDataInterfaceSkeletalMesh_0}
   }
  }]
}
